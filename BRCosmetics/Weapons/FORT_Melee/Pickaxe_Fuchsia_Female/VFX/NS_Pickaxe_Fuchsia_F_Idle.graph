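# Niagara idle effect for the Fuchsia female pickaxe, adding electric mesh arcs and a pulsing glow in the user variant color.
# Short-lived arc meshes sample the skeletal mesh and slide along a local offset while shrinking and fading.
# A persistent two-renderer mesh and a light, flickering through layered waveforms, also follow the skeletal mesh.
# Both mesh emitters boost brightness in the front end.
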NiagaraSystem /BRCosmetics/Weapons/FORT_Melee/Pickaxe_Fuchsia_Female/VFX/NS_Pickaxe_Fuchsia_F_Idle
dependencies {
 usedDataInterfaces [&NiagaraDataInterfaceCurve]
 usedDynamicInputs [&Lerp_Float, &SetBoolByFloatComparison]
 usedModules [
  &SystemState
  /Game/Effects/Niagara/Modules/Skydiving/FN_DivingAttributes/v4/FN_DivingAttributes_v4
  &self:SystemSpawnScript.SystemScriptSource.SystemScriptGraph.NiagaraNodeAssignment_0.SetVariables_#1
  &FN_Integrator
 ]
}
emitters {
 ElectricMeshArcsBody001 {
  emitterUpdateScript [
   &EmitterState(
    Life Cycle Mode=::NewEnumerator1
    Loop Duration=&UniformRangedFloat(Minimum=0.3, Maximum=0.5)
    Recalculate Duration Each Loop=true
    UseLoopDelay=false
   )
   &SpawnRate(SpawnRate=&RandomRangeFloat(Minimum=0, Maximum=1), Use Spawn Probability=false)
  ]
  particleSpawnScript [
   /Niagara/Modules/Spawn/Initialization/InitializeParticle(
    Color=$User.VariantColor
    Lifetime=&UniformRangedFloat(Minimum=0.15, Maximum=0.2)
    Mesh Scale=&MakeVector(
     X=&MaskFloatBySpawnGroup(Masked Float=-2, Passthrough Float=2, Spawn Group Mask=&RandomRangeInt(Minimum=0, Maximum=1))
     Y=-1
     Z=2.2
    )
    Module.Write Color=true
    Module.Write Lifetime=true
    Module.Write Mass=true
    Module.Write Material Random=true
    Module.Write Position=false
    Module.Write Scale=true
    Module.Write SpriteRotation=false
    Module.Write SpriteSize=false
    Sprite Size=<1 1>
   )
   &DynamicMaterialParameters(
    Index 0 Param 1=&UniformRangedFloat(Minimum=0.2, Maximum=0.1)
    Index 0 Param 2=&UniformRangedFloat(Minimum=0.4, Maximum=0.3)
    Index 0 Param 3=&UniformRangedFloat(Minimum=0, Maximum=1)
    Index 0 Param 4=&RandomRangeFloat(Minimum=0, Maximum=1)
    Index 1 Param 1=0.01
    Index 1 Param 2=&RandomRangeFloat(Minimum=0.01, Maximum=0)
    Module.Param0WriteEnabled=true
    Module.Param0WriteEnabled1=true
    Module.Param0WriteEnabled2=true
    Module.Param1WriteEnabled=true
    Module.Param1WriteEnabled1=true
    Module.Param1WriteEnabled2=true
    Module.Param2WriteEnabled=true
    Module.Param2WriteEnabled1=false
    Module.Param2WriteEnabled2=true
    Module.Param3WriteEnabled=true
    Module.Param3WriteEnabled1=false
    Module.Param3WriteEnabled2=true
    Write Parameter Index 1=true
    Write Parameter Index 2=true
   )
   &InitialMeshOrientation(
    Mesh Orientation Mode=::NewEnumerator2
    Orientation Coordinate Space=::Local
    Rotation=&RandomRangeVector(Minimum=<0 0 0>, Maximum=<0 0 1>)
    Rotation Coordinate Space=::NewEnumerator2
    Use Rotation Vector=true
   )
   /SetVariables(Particles.Scale=&Multiply_Vector(A=$Particles.Scale, B=<1 1 2>)) {isSetParameters=true}
  ]
  particleUpdateScript [
   &ParticleState()
   &ScaleMeshSize(Scale Factor=&Multiply_VectorByFloat(
     Vector=<1 1 1>
     Float=&FloatFromCurve(FloatCurve=DataInterface() {propertyValues {Curve {keys [{arriveTangent=-1.0067, interpMode=RCIM_Cubic, leaveTangent=-1.0067, tangentMode=RCTM_User, time=0, value=1.5}, <at 1 0.4>]}}})
    ))
   &UpdateMeshOrientation(Rotation Rate=5, Rotation Vector=<1 1 1>)
   &SkeletalMeshLocation(
    Apply Sampled Orientation to Mesh Particles=true
    Bone Sampling Mode=::NewEnumerator4
    Inherit Velocity=false
    Interpolate Mesh Sampling=false
    MaskSpawnGroup=false
    Sampled Position Offset=&Lerp_Vector(A=<0 0 0>, B=<80 0 0>, Alpha=$Particles.NormalizedAge)
    Sampled Position Offset Coordinate Space:ENiagaraMeshTransforms=::NewEnumerator2
    Skeletal Mesh=$User.SkeletalMesh
   )
   &ScaleColor(
    Scale Alpha=&FloatFromCurve(FloatCurve=DataInterface() {propertyValues {Curve {keys [<at 0.75 1 RCIM_Cubic>,<at 1 0>]}}}, Scale Curve=0.5)
    Scale RGB=&Multiply_VectorByFloat(Vector=<1 1 1>, Float=&FN_FrontEndFloatModifier(Gameplay=1, Front End=2))
    ScaleA=true
    ScaleRGB=true
   )
   /SetVariables(Particles.MeshOrientation=&MultiplyQuaternion(Quaternion A=$Output.SkeletalMeshLocation.SampledRotation, Quaternion B=$Particles.Initial.MeshOrientation)) {enabled=false, isSetParameters=true}
   &DynamicMaterialParameters(
    Index 0 Param 1=&FloatFromCurve(FloatCurve=DataInterface() {propertyValues {Curve {keys [<at 0 0.2>,<at 0.1 1.125>,<at 0.3 1 RCIM_Cubic>,<at 1 0>]}}})
    Index 1 Param 1=&FloatFromCurve(
     FloatCurve=DataInterface() {propertyValues {Curve {keys [{tangentMode=RCTM_User, time=0, value=1}, {tangentMode=RCTM_User, time=1, value=0}]}}}
     Scale Curve=0.3
    )
    Index 2 Param 1=1
    Index 2 Param 2=1
    Index 2 Param 3=1
    Index 2 Param 4=1
    Module.Param0WriteEnabled=true
    Module.Param0WriteEnabled1=true
    Module.Param0WriteEnabled2=true
    Module.Param1WriteEnabled=false
    Module.Param1WriteEnabled1=false
    Module.Param1WriteEnabled2=true
    Module.Param2WriteEnabled=false
    Module.Param2WriteEnabled1=false
    Module.Param2WriteEnabled2=true
    Module.Param3WriteEnabled=false
    Module.Param3WriteEnabled1=false
    Module.Param3WriteEnabled2=true
    Write Parameter Index 1=true
    Write Parameter Index 2=true
   ) {module=DynamicMaterialParameters001}
   &CameraOffset(Camera Offset Amount=15)
  ]
  rendererClasses [&NiagaraMeshRendererProperties]
  renderers [{rendererClass=&NiagaraMeshRendererProperties, rendererIndex=0}]
 }
 ElectricMeshArcsUnder001 {
  emitterUpdateScript [
   &EmitterState(
    Inactive Response=::NewEnumerator1
    Life Cycle Mode=::NewEnumerator1
    Loop Behavior=::NewEnumerator1
    Loop Duration Mode=::NewEnumerator1
    UseLoopDelay=false
   )
   &SpawnBurst_Instantaneous(Spawn Count=1, Use Spawn Probability=false)
  ]
  particleSpawnScript [
   /SetVariables(Particles.InitRand=HlslExpression() {hlslExpression="sign(rand(1.0)-0.5)"}) {isSetParameters=true}
   &InitializeParticle(
    Color=$User.VariantColor
    Mesh Scale=&Multiply_VectorByFloat(Vector=&MakeVector(X=4, Y=4, Z=4), Float=1)
    Module.Write Color=true
    Module.Write Lifetime=true
    Module.Write Mass=false
    Module.Write Material Random=true
    Module.Write Position=false
    Module.Write Scale=true
    Module.Write SpriteRotation=false
    Module.Write SpriteSize=false
    Sprite Size=<1 1>
   )
   &DynamicMaterialParameters(
    Index 0 Param 1=0.7
    Index 1 Param 1=1
    Module.Param0WriteEnabled=true
    Module.Param0WriteEnabled1=true
    Module.Param0WriteEnabled3=true
    Module.Param1WriteEnabled=false
    Module.Param1WriteEnabled1=false
    Module.Param1WriteEnabled3=true
    Module.Param2WriteEnabled=false
    Module.Param2WriteEnabled1=false
    Module.Param2WriteEnabled3=false
    Module.Param3WriteEnabled=false
    Module.Param3WriteEnabled1=false
    Module.Param3WriteEnabled3=false
    Write Parameter Index 1=true
    Write Parameter Index 3=true
   )
  ]
  particleUpdateScript [
   &ParticleState(Kill Particles When Lifetime Has Elapsed=false)
   &SkeletalMeshLocation(
    Apply Sampled Orientation to Mesh Particles=true
    Bone Sampling Mode=::NewEnumerator4
    Inherit Velocity=false
    Interpolate Mesh Sampling=false
    MaskSpawnGroup=false
    Sampled Position Offset=<5 0 0>
    Sampled Position Offset Coordinate Space:ENiagaraMeshTransforms=::NewEnumerator2
    Skeletal Mesh=$User.SkeletalMesh
   )
   &ScaleColor(
    Scale Alpha=2
    Scale RGB=&Multiply_VectorByFloat(Vector=<2 2 2>, Float=&FN_FrontEndFloatModifier(Gameplay=1, Front End=1.7))
    ScaleA=true
    ScaleRGB=true
   )
   &CameraOffset(Camera Offset Amount=1)
  ]
  rendererClasses [&NiagaraMeshRendererProperties]
  renderers [
   {rendererClass=&NiagaraMeshRendererProperties, rendererIndex=0}
   {rendererClass=&NiagaraMeshRendererProperties, rendererIndex=1}
  ]
 }
 Lig {
  emitterSpawnScript [&FN_PreserveSkydiveUserVariables() {module=PreserveSkydiveUserVariables}]
  emitterUpdateScript [
   &EmitterState(Life Cycle Mode=::NewEnumerator1, Loop Behavior=::NewEnumerator1, Loop Duration Mode=::NewEnumerator1, UseLoopDelay=false)
   &SpawnBurst_Instantaneous(Use Spawn Probability=false, UseLoopCountLimit=false) {module=SpawnBurst_Instantaneous001}
  ]
  particleSpawnScript [
   &InitializeParticle(
    Color=<rgba 3 0.6228 2.0597 1>
    Mesh Scale=<1.5 1.7 1.7>
    Module.Write Color=true
    Module.Write Lifetime=true
    Module.Write Mass=false
    Module.Write Material Random=true
    Module.Write Position=false
    Module.Write Scale=false
    Module.Write SpriteRotation=false
    Module.Write SpriteSize=false
    Position=&SimulationPosition()
    Sprite Rotation=&UniformRangedFloat(Minimum=0, Maximum=1)
   )
   /SetVariables(Particles.FloatID=HlslExpression() {hlslExpression=Particles.ID.Index}) {isSetParameters=true}
   /SetVariables(Particles.OffsetID:NiagaraFloat=&Multiply_Float(A=$Particles.FloatID, B=0.5)) {isSetParameters=true}
  ]
  particleUpdateScript [
   &ParticleState(Kill Particles When Lifetime Has Elapsed=false)
   &SkeletalMeshLocation(
    Apply Sampled Orientation to Mesh Particles=true
    Inherit Velocity=false
    Interpolate Mesh Sampling=false
    MaskSpawnGroup=false
    Mesh Sampling Type=::NewEnumerator5
    Sampled Position Offset=<80 0 0>
    Sampled Position Offset Coordinate Space:ENiagaraMeshTransforms=::NewEnumerator2
    Skeletal Mesh=$User.SkeletalMesh
    Socket Sampling Mode=::NewEnumerator5
   )
   /Game/Effects/Niagara/Modules/Skydiving/FN_DivingAttributes/FN_DivingAttributes_v2(Accumulate Dive=0.1, Accumulate Rest=0.1, ParaGlideLeanAlpha:NiagaraFloat=$User.ParaGlideLeanAlpha) {module=FN_DivingAttributes_v2001}
   &SolveForcesAndVelocity(Clamp Velocity=false, Limit Acceleration=false)
   &FN_Light_Attributes(
    Color=&MultiplyLinearColor(
     Linear Color=&MultiplyLinearColor(Linear Color=$User.VariantColor, Scale Factor=<1 1 1 1>)
     Scale Factor=&MakeVector4_FromFloat(Float=&Waveform(
       Waveform Count=::NewEnumerator2
       Global Phase=&Add_Float(A=$Emitter.Age, B=$Particles.OffsetID)
       "Waveform [1]"=::NewEnumerator0
       "[1] Phase Scale"=0.89
       "[1] Phase"=&Add_Float(A=$Emitter.Age, B=$Particles.OffsetID)
       "[1] Frequency"=1
       "[1] Amplitude Min/Max"=<0 1>
       "[2] Average"=false
       "[2] Input Position"=0
       "[2] Phase"=&Add_Float(A=$Emitter.Age, B=$Particles.OffsetID)
       "[2] Phase Scale"=1.3
       "[2] Compound Frequency"=2
       "[2] Frequency"=2
       "[2] Amplitude Min/Max"=<0.5 1>
       "[2] Scale"=1
       "[2] Bias"=0
       "[2] Exponent"=1
       "[2] Pulse Duration"=0.1
       "[2] Clip Triangle Min/Max"=<0 0>
       "[2] Random Seed"=0
       "[3] Average"=false
       "[3] Input Position"=0
       "[3] Phase"=&Add_Float(A=$Emitter.Age, B=$Particles.OffsetID)
       "[3] Phase Scale"=3
       "[3] Frequency"=0.5
       "[3] Compound Frequency"=2
       "[3] Amplitude Min/Max"=<0.5 1>
       "[3] Scale"=1
       "[3] Bias"=0
       "[3] Exponent"=1
       "[3] Pulse Duration"=0.1
       "[3] Clip Triangle Min/Max"=<0 0>
       "[3] Random Seed"=0
      ))
    )
    Create Color Attribute=true
    Create Exponent Attribute=true
    Create Position Attribute=true
    Create Radius Attribute=true
    Exponent=2
    Position:NiagaraPosition=$Particles.Position
    Radius=65
    Scale Color=0.4
   )
  ]
  rendererClasses [&NiagaraLightRendererProperties]
  renderers [{rendererClass=&NiagaraLightRendererProperties, rendererIndex=0}]
 }
}
summary {
 systemName=NS_Pickaxe_Fuchsia_F_Idle
 emitters [
  {emitterName=ElectricMeshArcsUnder001, rendererClasses [&NiagaraMeshRendererProperties]}
  {emitterName=ElectricMeshArcsBody001, rendererClasses [&NiagaraMeshRendererProperties]}
  {emitterName=Lig, rendererClasses [&NiagaraLightRendererProperties]}
 ]
 userVariables [
  User.bIsFrontEnd {type=NiagaraBool}
  User.bIsFrontEndPreview {type=NiagaraBool}
  User.ParaGlideLeanAlpha {type=NiagaraFloat, defaultValue=0, description=""}
  User.PlayerSpeed {type=NiagaraFloat, defaultValue=0, description=""}
  User.SkeletalMesh {
   type=NiagaraDataInterfaceSkeletalMesh
   defaultValue {
    struct=NiagaraExt_VariableValue_DataInterface
    value {dataInterface=&self:NiagaraDataInterfaceSkeletalMesh_0}
   }
  }
  User.VariantColor {type=LinearColor, description="", defaultValue {struct=LinearColor, value=<rgba 0.2038 0.0293 1 1>}}
 ]
}
